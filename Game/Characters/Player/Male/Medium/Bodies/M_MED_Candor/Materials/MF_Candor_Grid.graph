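# Material function for the Candor player body that layers a glowing, glitchy digital grid effect over incoming material attributes.
# The isStageTwo static switch selects between two looks, each lerped into BaseColor and added to Emissive.
# MaterialCurveA times PhaseSpeed advances the reveal masks and their thin scanline bands, with a grid trail in stage two.
# Rotatable UV grid, Fresnel, pre-skinned bounds gradients, time-driven noise distortion and pulse supply the detail through named reroutes.
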MaterialFunction /Game/Characters/Player/Male/Medium/Bodies/M_MED_Candor/Materials/MF_Candor_Grid
graph {nodes [
  n1 = GetMaterialAttributes(None=FunctionInput() {inputType=FunctionInput_MaterialAttributes}) {attributeGetTypes [#1, #2]}
  n2 = Add(A=@n1.EmissiveColor, B=NamedRerouteUsage())
  n3 = Add(A=@n1.EmissiveColor, B=NamedRerouteUsage())
  n4 = LinearInterpolate(A=@n1.BaseColor, B=NamedRerouteUsage(), Alpha=NamedRerouteUsage())
  n5 = LinearInterpolate(A=@n1.BaseColor, B=NamedRerouteUsage(), Alpha=NamedRerouteUsage())
  FunctionOutput(None=SetMaterialAttributes!(
    MaterialAttributes=@n1.MaterialAttributes
    EmissiveColor=StaticSwitchParameter(True=@n2, False=@n3) {group=Candor, parameterName=isStageTwo}
    BaseColor=StaticSwitchParameter(True=@n4, False=@n5) {group=Candor, parameterName=isStageTwo}
   ) {attributeSetTypes [#1, #2]})
  n6 = VectorParameter() {group=Candor Stage 1, parameterName=Grid_Color, defaultValue=<rgba 0.2384 0.5395 0.1095 1>}
  n7 = NamedRerouteUsage()
  n8 = NamedRerouteUsage()
  n9 = Add(A=Multiply(A=@n7, B=@n8), B=Multiply(A=NamedRerouteUsage(), B=NamedRerouteUsage()))
  n10 = NamedRerouteUsage()
  n11 = Multiply(A=Multiply(A=@n6.RGB, B=Multiply(A=@n9, B=@n10)), B=ScalarParameter() {defaultValue=2.5, group=Candor Stage 1, parameterName=Grid_Intensity})
  n12 = NamedRerouteUsage()
  NamedRerouteDeclaration(None=Multiply(A=Add(A=@n11, B=@n12), B=NamedRerouteUsage())) {name=Stage1 Emissive Out, nodeColor=<rgba 0.7578 1 0 1>}
  n13 = MaterialFunctionCall!(Time, Divisor=Constant() {r=0.333}, Mod=Constant() {r=1000}) {materialFunction=/Game/Effects/Fort_Effects/Materials/_Functions/MF_ModuloTime}
  n14 = ComponentMask(None=MaterialFunctionCall!(
    PrimaryTextureCoordinates=TextureCoordinate() {coordinateIndex=1, vTiling=50}
    2ndOrderTextureCoordinates
    RandomSeed=Round(None=@n13.Unclamped)
    NoiseDensityBias
    Bypass Noise Influence Crosstalk
   ) {materialFunction=/Game/Effects/Fort_Effects/Materials/_Functions/MF_RandomUVSample}.NoiseWithRange) {g=true}
  n15 = Constant()
  n16 = Saturate(None=Time())
  n17 = Power(Base=Max(A=Sine(None=@n16) {period=3.6}, B=Sine(None=@n16) {period=2.44})) {constExponent=10}
  n18 = Add(A=Multiply(A=AppendVector(A=@n14, B=@n15), B=Round(None=@n17))) {constB=-0.5}
  n19 = Saturate(None=Multiply(A=@n18, B=ScalarParameter() {defaultValue=0.05, group=Candor, parameterName=Distortion Scale}))
  NamedRerouteDeclaration(None=Add(
    A=TextureSample(UVs=@n19) {samplerType=SAMPLERTYPE_LinearColor, texture=/Game/Global/Textures/Default/Noise/T_Atlas_DigitalNoise_01}.G
    B=Constant() {r=0.25}
   )) {name=Glitch, nodeColor=<rgba 1 0.8 0 1>}
  n20 = PreSkinnedLocalBounds()
  n21 = Subtract(A=NamedRerouteUsage(), B=Add(A=@n20.Min) {constB=-3})
  n22 = Add(A=@n20.Extents) {constB=15}
  NamedRerouteDeclaration(None=MaterialFunctionCall(Float3=Divide(A=@n21, B=@n22)) {materialFunction=&BreakOutFloat3Components}.B) {name=CandorLocalBounds, nodeColor=<rgba 1 0.7734 0 1>}
  n23 = NamedRerouteUsage()
  n24 = ScalarParameter() {defaultValue=0.05, group=Candor Stage 2, parameterName=Stage2 Glitch Opacity}
  n25 = NamedRerouteUsage()
  n26 = ScalarParameter() {defaultValue=0.05, group=Candor Stage 2, parameterName=Stage2 Digital Noise Opacity}
  n27 = Add(
   A=Multiply(A=Multiply(A=@n23, B=@n24), B=Multiply(A=@n25, B=@n26))
   B=LinearInterpolate(
    A=VectorParameter() {group=Candor Stage 2, parameterName=Stage2 Edge Color, defaultValue=<rgba 0 0.0167 0.0365 1>}.RGB
    B=VectorParameter() {group=Candor Stage 2, parameterName=Stage2 Center Color, defaultValue=<rgba 0 0.1543 0.6458 1>}.RGB
    Alpha=NamedRerouteUsage()
   )
  )
  n28 = NamedRerouteUsage()
  NamedRerouteDeclaration(None=Multiply(A=@n27, B=@n28)) {name=Stage2 Color Out, nodeColor=<rgba 1 0 0.9375 1>}
  n29 = ComponentMask(None=NamedRerouteUsage()) {b=true, r=true}
  n30 = Multiply(A=NamedRerouteUsage(), B=@n29)
  n31 = NamedRerouteUsage()
  n32 = TextureSample(UVs=Subtract(A=Multiply(A=@n30) {constB=0.025}, B=@n31)) {samplerType=SAMPLERTYPE_LinearColor, texture=&T_Atlas_DigitalNoise_01}
  n33 = ScalarParameter() {defaultValue=1, group=Candor, parameterName=Digital Noise Opacity}
  n34 = Add(A=@n32.B, B=@n33)
  n35 = Time()
  n36 = Divide(A=@n35, B=Constant() {r=3})
  n37 = Ceil(None=Subtract(A=Sine(None=@n36), B=Constant() {r=0.9}))
  n38 = ScalarParameter() {defaultValue=0.6, group=Candor, parameterName=Digital Noise Frequency, sliderMax=1}
  n39 = Sine(None=Divide(A=@n35, B=Constant() {r=1}))
  n40 = Constant() {r=0.6}
  n41 = Saturate(None=Ceil(None=Subtract(A=@n39, B=@n40)))
  n42 = Add(A=Multiply(A=Saturate(None=@n37), B=@n38), B=OneMinus(None=Multiply(A=@n41, B=@n38)))
  n43 = Multiply(A=Multiply(A=Subtract(A=@n34, B=@n32.R), B=@n42), B=Clamp(None=Time()) {maxDefault=0.3, minDefault=0.02})
  n44 = ScalarParameter() {defaultValue=0.35, group=Candor, parameterName=Digital Noise Intensity}
  NamedRerouteDeclaration(None=Saturate(None=Multiply(A=@n43, B=@n44) {constB=0.15})) {name=Digital Noise, nodeColor=<rgba 0.0547 0 1 1>}
  n45 = Saturate(None=Fresnel(ExponentIn=ScalarParameter() {defaultValue=3, group=Candor, parameterName=Grid Fresnel}) {baseReflectFraction=0})
  NamedRerouteDeclaration(None=@n45) {name=Grid Mask Fresnel, nodeColor=<rgba 0 0.3 1 1>}
  n46 = Multiply(A=@n27, B=@n28)
  n47 = NamedRerouteUsage()
  n48 = NamedRerouteUsage()
  n49 = NamedRerouteUsage()
  n50 = VectorParameter() {group=Candor, parameterName=Scanline Color, defaultValue=<rgba 0 0.7157 0.0273 1>}
  n51 = Multiply(
   A=Multiply(A=@n50.RGB, B=Multiply(A=@n48, B=@n49))
   B=ScalarParameter() {defaultValue=5, group=Candor Stage 2, parameterName=Grid Trail Opacity}
  )
  NamedRerouteDeclaration(None=Add(A=Add(A=@n46, B=@n47), B=Saturate(None=@n51))) {name=Stage2 Emissive Out, nodeColor=<rgba 0.9754 1 0 1>}
  n52 = Add(A=@n7, B=@n8)
  NamedRerouteDeclaration(None=Multiply(A=Multiply(A=@n6.RGB, B=@n52), B=NamedRerouteUsage())) {name=Stage1 Color Out, nodeColor=<rgba 1 0 0.8203 1>}
  n53 = Saturate(None=Multiply(
    A=ScalarParameter() {group=Candor, parameterName=MaterialCurveA}
    B=ScalarParameter() {defaultValue=5, group=Candor, parameterName=PhaseSpeed}
   ))
  n54 = Add(A=Subtract(A=NamedRerouteUsage()), B=@n53)
  n55 = Saturate(None=@n54)
  n56 = Ceil(None=@n55)
  NamedRerouteDeclaration(None=@n56) {name=Stage2 Mask, nodeColor=<rgba 1 0.9844 0 1>}
  n57 = Ceil(None=Subtract(A=@n55, B=ScalarParameter() {defaultValue=0.015, group=Candor, parameterName=ScanLineThickness}))
  NamedRerouteDeclaration(None=Multiply(A=@n56, B=OneMinus(None=@n57))) {name=Stage2 Scanline Mask, nodeColor=<rgba 0 0.5547 1 1>}
  n58 = NamedRerouteUsage()
  NamedRerouteDeclaration(None=Multiply(A=Multiply(A=@n50.RGB, B=@n58), B=ScalarParameter() {defaultValue=5, group=Candor, parameterName=Scanline Intensity})) {name=Stage2 Scanline Out, nodeColor=<rgba 0 1 0.9062 1>}
  n59 = ScalarParameter() {defaultValue=0.01, group=Candor Stage 1, parameterName=Grid Distortion Intesnity}
  NamedRerouteDeclaration(None=Saturate(None=Multiply(A=@n18, B=@n59) {constB=0})) {name=GridGlitch, nodeColor=<rgba 0 0.6719 1 1>}
  n60 = NamedRerouteUsage()
  n61 = ComponentMask(None=PreSkinnedLocalBounds().Half Extents) {b=true}
  n62 = MaterialFunctionCall(Z=Add(A=@n61, B=ScalarParameter() {defaultValue=28, group=Candor Stage 2, parameterName=Stage2 Center Gradient Position}) {constB=28}) {materialFunction=&MakeFloat3}
  n63 = Power(
   Base=Divide(A=Distance(A=@n60, B=@n62.Result), B=@n61)
   Exp=ScalarParameter() {defaultValue=1.5, group=Candor Stage 2, parameterName=Center Gradient Power}
  ) {constExponent=0.8}
  NamedRerouteDeclaration(None=OneMinus(None=Saturate(None=@n63))) {name=Stage 2 Gradient Mask, nodeColor=<rgba 1 0 0.2344 1>}
  n64 = ScalarParameter() {defaultValue=0.245, group=Candor Feet UVs, parameterName=Degree Rotation}
  n65 = Add(
   A=Subtract(A=MaterialFunctionCall() {materialFunction=&BoundingBoxBased_0-1_UVW}.B)
   B=OneMinus(None=Constant() {desc=Gradient Height, r=0.125})
  )
  n66 = OneMinus(None=Ceil(None=Saturate(None=@n65)))
  n67 = TextureCoordinate()
  n68 = MaterialFunctionCall(
   X=ScalarParameter() {defaultValue=0.1, group=Candor Feet UVs, parameterName=UV X Rotation}
   Y=ScalarParameter() {defaultValue=0.6, group=Candor Feet UVs, parameterName=UV Y Rotation}
  ) {materialFunction=&MakeFloat2}
  n69 = Add(
   A=MaterialFunctionCall!(Angle=Multiply(A=@n64, B=@n66), UV/2DVector=Add(A=@n67, B=@n68.Result)) {materialFunction=/Game/Characters/Player/Male/Medium/Bodies/M_MED_Faux/Materials/MF_2DRotate}.Result
   B=NamedRerouteUsage()
  )
  n70 = ScalarParameter() {defaultValue=45, group=Candor, parameterName=Grid Size}
  n71 = Multiply(A=Frac(None=Multiply(A=@n69, B=@n70))) {constB=2}
  n72 = ScalarParameter() {defaultValue=0.2, group=Candor Stage 1, parameterName=Grid Thickness}
  n73 = Ceil(None=Saturate(None=Subtract(A=@n71, B=@n72)))
  n74 = MaterialFunctionCall(Vector 2=Saturate(None=OneMinus(None=@n73))) {materialFunction=/Engine/Functions/Engine_MaterialFunctions02/Utility/VectorLength}
  NamedRerouteDeclaration(None=@n74.V2 Length) {name=Grid Mask Solo, nodeColor=<rgba 1 0.5391 0 1>}
  NamedRerouteDeclaration(None=Saturate(None=Add(A=@n45, B=@n74.V2 Length))) {name=Stage2 Grid Fresnel, nodeColor=<rgba 0 1 0.6484 1>}
  n75 = Time()
  NamedRerouteDeclaration(None=Frac(None=Multiply(A=@n75) {constB=0.25})) {name=LocalDebug, nodeColor=<rgba 0 0.1797 1 1>}
  NamedRerouteDeclaration(None=Saturate(None=@n18)) {name=DigitalNoise Distortion, nodeColor=<rgba 0.7578 0 1 1>}
  n76 = NamedRerouteUsage()
  NamedRerouteDeclaration(None=Multiply(A=Multiply(A=@n50.RGB, B=@n76), B=ScalarParameter() {defaultValue=2.5, group=Candor, parameterName=ScanLine_Intensity})) {name=Stage1 Scanline Out, nodeColor=<rgba 0 1 0.9063 1>}
  n77 = OneMinus(None=@n56)
  n78 = Add(A=@n54, B=ScalarParameter() {defaultValue=0.08, group=Candor Stage 2, parameterName=Grid Trail Thickness})
  NamedRerouteDeclaration(None=Multiply(A=Multiply(A=@n77, B=@n78), B=ScalarParameter() {defaultValue=5, group=Candor Stage 2, parameterName=Grid Trail Opacity})) {name=Stage2 Grid Trail, nodeColor=<rgba 0 0.1328 1 1>}
  NamedRerouteDeclaration(None=VertexInterpolator(VS=PreSkinnedPosition()).PS) {name=MeshPosition, nodeColor=<rgba 0.1016 0 1 1>}
  n79 = Add(A=Subtract(A=NamedRerouteUsage()), B=@n53)
  n80 = Saturate(None=@n79)
  n81 = Ceil(None=@n80)
  n82 = Add(A=@n79, B=ScalarParameter() {defaultValue=0.15, group=Candor Stage 1, parameterName=Texture Reveal Thickness})
  NamedRerouteDeclaration(None=Multiply(A=Multiply(A=@n81, B=@n82), B=ScalarParameter() {defaultValue=3.5, group=Candor Stage 1, parameterName=TextureReveal Opacity})) {name=Stage1 Mask, nodeColor=<rgba 0 0.7656 1 1>}
  n83 = Ceil(None=Subtract(A=@n80, B=ScalarParameter() {defaultValue=0.015, group=Candor, parameterName=ScanLineThickness}))
  NamedRerouteDeclaration(None=Multiply(A=@n81, B=OneMinus(None=@n83))) {name=Stage1 Scanline Mask, nodeColor=<rgba 0 0.5547 1 1>}
  n84 = Sine(None=Multiply(A=ScalarParameter() {defaultValue=0.8, group=Candor, parameterName=Pulse Speed}, B=Time()))
  n85 = ScalarParameter() {defaultValue=0.7, group=Candor, parameterName=Pulse Strength}
  NamedRerouteDeclaration(None=Clamp(None=Frac(None=@n84), Min=OneMinus(None=@n85)) {minDefault=0.25}) {name=Pulse, nodeColor=<rgba 0.7344 0 1 1>}
  n86 = Multiply(A=@n29) {constB=0.025}
  n87 = Add(
   A=TextureSample(UVs=Subtract(A=@n86, B=@n31)) {samplerType=SAMPLERTYPE_LinearColor, texture=&T_Atlas_DigitalNoise_01}.B
   B=@n33
  )
  NamedRerouteDeclaration(None=Multiply(A=Multiply(A=@n87, B=@n42), B=ScalarParameter() {defaultValue=3, group=Candor Stage 2, parameterName=Distortion Intensity})) {name=Distortion, nodeColor=<rgba 1 0 0.1406 1>}
  NamedRerouteUsage()
  OneMinus(None=NamedRerouteUsage())
  Reroute(None=@n81) {desc=Solid Stage1 Mask}
 ]}
